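# SoundCue for the low-magazine variant of the Shotgun Siphon healing sound in the DoubleTrouble gameplay content.
# A single wave player feeds a modulator with volume held fixed, and the cue raises pitch slightly.
# It uses the consumable-use attenuation and consumables sound class.
# A concurrency override limits it to one instance per owner, stopping the oldest when retriggered.
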SoundCue /TidalCraneExotics/Gameplay/DoubleTrouble/Audio/SC_Shotgun_Siphon_Healing_LowMag
root=Modulator_0
nodes [
 {id=Modulator_0, children [WavePlayer_0], properties {volumeMax=1, volumeMin=1}}
 {id=WavePlayer_0, properties {soundWaveAssetPtr=&/Shotgun_Siphon_Healing_LowMag}}
]
properties {
 attenuationSettings=&Fort_Consumable_Use_Athena_Att
 bOverrideConcurrency=true
 maxDistance=2250
 pitchMultiplier=1.1
 soundClassObject=&Fort_Consumables
 concurrencyOverrides {bLimitToOwner=true, maxCount=1, resolutionRule=StopOldest}
}
